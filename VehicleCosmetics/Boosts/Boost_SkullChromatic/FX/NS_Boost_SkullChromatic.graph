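# Niagara system for the SkullChromatic vehicle boost cosmetic, placed at the car offset plus boost position.
# A main multi-renderer ribbon and boost-cone ribbon form the trail, joined by burst-spawned skull meshes and smoke sprites.
# Skull and smoke emitters use spawn groups to vary placement and set visibility tags.
# Painted-variant modules supply color channels, mostly red, green and blue for the chromatic split, scaled by curves; user bools adjust painted and frontend-preview looks.
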NiagaraSystem /VehicleCosmetics/Boosts/Boost_SkullChromatic/FX/NS_Boost_SkullChromatic
dependencies {
 usedDataInterfaces [&NiagaraDataInterfaceCurve, &NiagaraDataInterfaceSkeletalMesh]
 usedDynamicInputs [&MaskVector2dBySpawnGroup, &RandomRangeVector2D]
 usedModules [
  &self:SystemSpawnScript.SystemScriptSource.SystemScriptGraph.NiagaraNodeAssignment_0.SetVariables_#1
  &SystemState
 ]
}
emitters {
 NE_MainRibbon002 {
  emitterUpdateScript [
   &EmitterState(Life Cycle Mode=::NewEnumerator1, Loop Duration=0.5, UseLoopDelay=false)
   &VC_FilteredSocketCount()
   &VC_RateQualityHandler(Epic Spawn Rate=30, High Spawn Rate=20, Low Spawn Rate=20, Medium Spawn Rate=20, Use Custom Rate?=true)
   &VC_SpawnRibbon_v1(SpawnRate=$Emitter.Spawn Rate)
  ]
  particleSpawnScript [
   &VC_PaintedVariantsOverride(
    Black=<rgba 0.1 0.1 0.1 10>
    BurntSienna=<rgba 0.5 0.1175 0.05 1>
    Cobalt=<rgba 0.1 0.1675 1 1>
    Crimson=<rgba 0.35 0.0299 0.0245 1.5>
    ForestGreen=<rgba 0.2923 0.5 0.2523 1>
    Grey=<rgba 0.14 0.14 0.14 1.5>
    Lime=<rgba 0.6771 1.25 0.1562 1>
    Orange=<rgba 1.5 0.5975 0.075 1>
    Pink=<rgba 0.45 0.135 0.4342 1>
    Purple=<rgba 0.4016 0.1125 1.5 1.5>
    R1 Black:LinearColor=<rgba 1 0 0 1>
    R1 BurntSienna:LinearColor=<rgba 1 0 0 1>
    R1 Cobalt:LinearColor=<rgba 1 0 0 1>
    R1 Crimson:LinearColor=<rgba 1 0 0 1>
    R1 DefaultColor=<rgba 1 0 0 1>
    R1 ForestGreen:LinearColor=<rgba 1 0 0 1>
    R1 Grey:LinearColor=<rgba 1 0 0 1>
    R1 Lime:LinearColor=<rgba 1 0 0 1>
    R1 Orange:LinearColor=<rgba 1 0 0 1>
    R1 Pink:LinearColor=<rgba 1 0 0 1>
    R1 Purple:LinearColor=<rgba 1 0 0 1>
    R1 Saffron:LinearColor=<rgba 1 0 0 1>
    R1 SkyBlue:LinearColor=<rgba 1 0 0 1>
    R1 TitaniumWhite:LinearColor=<rgba 1 0 0 1>
    R2 BurntSienna:LinearColor=<rgba 0.15 0.0319 0.015 1>
    R2 Cobalt:LinearColor=<rgba 0.07 0.1398 1 1>
    R2 Crimson:LinearColor=<rgba 0.35 0.0299 0.0245 1.5>
    R2 DefaultColor:LinearColor=<rgba 0 1 0 1>
    R2 ForestGreen:LinearColor=<rgba 0.028 0.45 0.0135 1.5>
    R2 Grey:LinearColor=<rgba 0.14 0.14 0.14 1.5>
    R2 Lime:LinearColor=<rgba 0.2568 0.5391 0 1>
    R2 Orange:LinearColor=<rgba 0.5391 0.1617 0 1>
    R2 Pink:LinearColor=<rgba 0.45 0.135 0.4342 1>
    R2 Purple:LinearColor=<rgba 0.0812 0.0105 0.35 1.5>
    R2 Saffron:LinearColor=<rgba 0.35 0.35 0 1>
    R2 SkyBlue:LinearColor=<rgba 0.0237 0.3848 0.475 1>
    R2 TitaniumWhite:LinearColor=<rgba 0.643 0.643 0.643 0.75>
    R3 BurntSienna:LinearColor=<rgba 0.025 0.0059 0.0031 1>
    R3 Cobalt:LinearColor=<rgba 0.0075 0.0257 0.25 2>
    R3 Crimson:LinearColor=<rgba 0.35 0.0299 0.0245 1.5>
    R3 DefaultColor:LinearColor=<rgba 0 0 1 1>
    R3 ForestGreen:LinearColor=<rgba 0.028 0.45 0.0135 1.5>
    R3 Grey:LinearColor=<rgba 0.14 0.14 0.14 1.5>
    R3 Lime:LinearColor=<rgba 0.2568 0.5391 0 1>
    R3 Orange:LinearColor=<rgba 0.5391 0.1617 0 1>
    R3 Pink:LinearColor=<rgba 0.45 0.135 0.4342 1>
    R3 Purple:LinearColor=<rgba 0.0278 0.0036 0.1198 2>
    R3 Saffron:LinearColor=<rgba 0.35 0.35 0 2>
    R3 SkyBlue:LinearColor=<rgba 0.0237 0.3848 0.475 1>
    R3 TitaniumWhite:LinearColor=<rgba 0.55 0.55 0.55 1>
    Saffron=<rgba 1 1 0.287 1>
    SkyBlue=<rgba 0.36 0.792 0.9 1>
    TitaniumWhite=<rgba 0.7 0.7 0.7 0.75>
    Use R1 Painted Overrides=true
    Use R2 Painted Overrides=true
    Use R3 Painted Overrides=true
    Use R4 Painted Overrides=true
   ) {enabled=false}
   &VC_InitializeRibbon_v1(
    Color=$Particles.Color
    Custom Offset=true
    Lifetime=1
    Local Vector Offset=&AddVectorToPosition(
     Position=&AddVectorToPosition(Position=$User.DTCarOffset, Vector=Linked() {linkedVariable {name=User.SingleBoostPos, type=NiagaraPosition}})
     Vector=<-110 0 0>
    )
    Offset Amount=1
    Ribbon Width=70
   )
   &AddVelocity(Velocity=&MakeCustomVectorFromBool(Bool=$User.bIsFrontendPreview, True Vector=<-1500 0 0>, False Vector=<-1000 0 0>)) {module=AddVelocity001}
  ]
  particleUpdateScript [
   &ParticleState()
   &InheritSourceMovement(
    Applied Position Scale=&VectorFromFloat(Value=&FloatFromCurve(FloatCurve=DataInterface() {propertyValues {Curve {keys [
          {interpMode=RCIM_Cubic, tangentMode=RCTM_Break, time=0, value=1}
          {arriveTangent=-2.9025, interpMode=RCIM_Cubic, tangentMode=RCTM_Break, time=1, value=0}
         ]}}}))
    ApplyForce=false
    ApplyPosition=true
    ApplyPositionFalloff=false
   )
   &SolveForcesAndVelocity(Clamp Velocity=false, Limit Acceleration=false)
   &VC_PaintedVariantsOverride(
    Black=<rgba 0.25 0.25 0.25 1>
    BurntSienna=<rgba 0.7 0.1645 0.07 1>
    Cobalt=<rgba 0.3 0.5833 2 1>
    Crimson=<rgba 1 0.15 0.15 0.8>
    ForestGreen=<rgba 0.4092 0.7 0.3533 1>
    Grey=<rgba 0.8 0.8 0.8 0.4>
    Lime=<rgba 0.7262 1.25 0.25 1>
    Orange=<rgba 1 0.6267 0.2 0.7>
    Pink=<rgba 0.7833 0.35 1 1>
    Purple=<rgba 0.4505 0.17 1.7 1.5>
    R1 Black:LinearColor=<rgba 1 0 0 1>
    R1 BurntSienna:LinearColor=<rgba 1 0 0 1>
    R1 Cobalt:LinearColor=<rgba 1 0 0 1>
    R1 Crimson:LinearColor=<rgba 1 0 0 1>
    R1 DefaultColor=<rgba 1 0 0 1>
    R1 ForestGreen:LinearColor=<rgba 1 0 0 1>
    R1 Grey:LinearColor=<rgba 1 0 0 1>
    R1 Lime:LinearColor=<rgba 1 0 0 1>
    R1 Orange:LinearColor=<rgba 1 0 0 1>
    R1 Pink:LinearColor=<rgba 1 0 0 1>
    R1 Purple:LinearColor=<rgba 1 0 0 1>
    R1 Saffron:LinearColor=<rgba 1 0 0 1>
    R1 SkyBlue:LinearColor=<rgba 1 0 0 1>
    R1 TitaniumWhite:LinearColor=<rgba 1 0 0 0>
    R2 Black:LinearColor=<rgba 0 1 0 1>
    R2 BurntSienna:LinearColor=<rgba 0 1 0 1>
    R2 Cobalt:LinearColor=<rgba 0 1 0 1>
    R2 Crimson:LinearColor=<rgba 0 1 0 1>
    R2 DefaultColor:LinearColor=<rgba 0 1 0 1>
    R2 ForestGreen:LinearColor=<rgba 0 1 0 1>
    R2 Gold:LinearColor=<rgba 0 1 0 1>
    R2 Grey:LinearColor=<rgba 0 1 0 1>
    R2 Lime:LinearColor=<rgba 0 1 0 1>
    R2 Orange:LinearColor=<rgba 0 1 0 1>
    R2 Pink:LinearColor=<rgba 0 1 0 1>
    R2 Purple:LinearColor=<rgba 0 1 0 1>
    R2 Saffron:LinearColor=<rgba 0 1 0 1>
    R2 SkyBlue:LinearColor=<rgba 0 1 0 1>
    R2 TitaniumWhite:LinearColor=<rgba 0 1 0 0>
    R3 Black:LinearColor=<rgba 0 0 1 1>
    R3 BurntSienna:LinearColor=<rgba 0 0 1 1>
    R3 Cobalt:LinearColor=<rgba 0 0 1 1>
    R3 Crimson:LinearColor=<rgba 0 0 1 1>
    R3 DefaultColor:LinearColor=<rgba 0 0 1 1>
    R3 ForestGreen:LinearColor=<rgba 0 0 1 1>
    R3 Gold:LinearColor=<rgba 0 0 1 1>
    R3 Grey:LinearColor=<rgba 0 0 1 1>
    R3 Lime:LinearColor=<rgba 0 0 1 1>
    R3 Orange:LinearColor=<rgba 0 0 1 1>
    R3 Pink:LinearColor=<rgba 0 0 1 1>
    R3 Purple:LinearColor=<rgba 0 0 1 1>
    R3 Saffron:LinearColor=<rgba 0 0 1 1>
    R3 SkyBlue:LinearColor=<rgba 0 0 1 1>
    R3 TitaniumWhite:LinearColor=<rgba 0 0 1 0>
    R4 Black:LinearColor=<rgba 0.07 0.07 0.07 0.75>
    R4 BurntSienna:LinearColor=<rgba 0.3 0.025 0 1>
    R4 Cobalt:LinearColor=<rgba 0 0.1 2 0.8>
    R4 Crimson:LinearColor=<rgba 0.8 0 0.02 0.7>
    R4 ForestGreen:LinearColor=<rgba 0.1 0.6 0 1>
    R4 Grey:LinearColor=<rgba 0.2 0.2 0.2 0.8>
    R4 Lime:LinearColor=<rgba 0.1667 0.5 0 1>
    R4 Orange:LinearColor=<rgba 0.6724 0.1121 0 1>
    R4 Pink:LinearColor=<rgba 0.4336 0.0758 0.5052 1>
    R4 Purple:LinearColor=<rgba 0.1246 0 0.6795 1>
    R4 Saffron:LinearColor=<rgba 0.8 0.7333 0 0.8>
    R4 SkyBlue:LinearColor=<rgba 0 0.4433 0.7 1>
    R4 TitaniumWhite:LinearColor=<rgba 2 2 2 0.5>
    Saffron=<rgba 1 1 0.4 1.5>
    SkyBlue=<rgba 0.36 0.792 0.9 1>
    TitaniumWhite=<rgba 1.75 1.75 1.75 0.2>
    Use R1 Painted Overrides=true
    Use R2 Painted Overrides=true
    Use R3 Painted Overrides=true
    Use R4 Painted Overrides=true
   ) {module=VC_PaintedVariantsOverride001}
   &ScaleRibbonWidth(Ribbon Width Scale=&FloatFromCurve(FloatCurve=DataInterface() {propertyValues {Curve {keys [
         <at 0 0.2>
         {arriveTangent=1.6, leaveTangent=1.6, time=0.2, value=0.8}
         {arriveTangent=-0.6154, leaveTangent=-0.6154, time=0.5, value=1}
         <at 1.5 0>
        ]}}}))
   /SetVariables(
    Particles.R1_DynamicMaterialParameters:Vector4f=&MakeVector4(
     X=&FloatFromCurve(FloatCurve=DataInterface() {propertyValues {Curve {keys [<at 0.5 0.7>,<at 1 1.5>]}}})
     Y=0
     Z=0.5
     W=0
    )
    Particles.R1_ScaledRibbonColor_R:LinearColor=&VC_DynamicScaleColor(
     Color to Scale:LinearColor=$Particles.R1 Color
     RGB Scale:Vector3f=&VectorFromFloat(Value=&MakeCustomFloatFromBool(Bool=$User.bIsPainted, True Float=1, False Float=2))
     Alpha Scale:NiagaraFloat=&MakeCustomFloatFromBool(
      Bool=$User.bIsPainted
      True Float=&FloatFromCurve(FloatCurve=DataInterface() {propertyValues {Curve {
          defaultValue=1
          keys [<at 0 0.2 RCIM_Cubic>, {arriveTangent=-0.25, interpMode=RCIM_Cubic, leaveTangent=-0.25, time=0.5, value=1.5}, <at 0.8 0 RCIM_Cubic>]
         }}})
      False Float=1.1
     )
    )
    Particles.R1_ScaledRibbonWIdth:NiagaraFloat=&Multiply_Float(
     A=70
     B=&FloatFromCurve(FloatCurve=DataInterface() {propertyValues {Curve {keys [<at 0 0.5>, {arriveTangent=1.6, time=0.1, value=0.8}, {arriveTangent=-0.6154, leaveTangent=-0.6154, time=0.5, value=1}, <at 1.5 0>]}}})
    )
    Particles.R2_ScaledRibbonColor_G:LinearColor=&VC_DynamicScaleColor(
     Color to Scale:LinearColor=$Particles.R2 Color
     RGB Scale:Vector3f=&VectorFromFloat(Value=2)
     Alpha Scale:NiagaraFloat=&MakeCustomFloatFromBool(
      Bool=$User.bIsPainted
      True Float=&FloatFromCurve(
       FloatCurve=DataInterface() {propertyValues {Curve {
          defaultValue=1
          keys [
           {interpMode=RCIM_Cubic, tangentMode=RCTM_User, time=0, value=0.3}
           {interpMode=RCIM_Cubic, tangentMode=RCTM_User, time=0.1, value=0}
           {tangentMode=RCTM_User, time=0.3, value=1}
           <at 0.8 1>
          ]
         }}}
       Scale Curve=1.1
      )
      False Float=1.1
     )
    )
    Particles.R3_ScaledRibbonColor_B:LinearColor=&VC_DynamicScaleColor(
     Color to Scale:LinearColor=$Particles.R3 Color
     RGB Scale:Vector3f=&VectorFromFloat(Value=2)
     Alpha Scale:NiagaraFloat=&MakeCustomFloatFromBool(
      Bool=$User.bIsPainted
      True Float=&FloatFromCurve(FloatCurve=DataInterface() {propertyValues {Curve {defaultValue=1, keys [<at 0 0.1 RCIM_Cubic>,<at 0.5 1 RCIM_Cubic>]}}}, Scale Curve=2)
      False Float=1.1
     )
    )
    Particles.SCALEPAINTEDS:LinearColor=&VC_DynamicScaleColor(
     Color to Scale:LinearColor=$Particles.R4 Color
     RGB Scale:Vector3f=<1 1 1>
     Alpha Scale:NiagaraFloat=&MakeCustomFloatFromBool(
      Bool=$User.bIsPainted
      True Float=&FloatFromCurve(
       FloatCurve=DataInterface() {propertyValues {Curve {
          defaultValue=1
          keys [
           {arriveTangent=4.1785, interpMode=RCIM_Cubic, leaveTangent=4.1785, tangentMode=RCTM_User, time=0, value=0}
           {arriveTangentWeight=0.05, interpMode=RCIM_Cubic, leaveTangentWeight=0.45, tangentMode=RCTM_User, time=0.1, value=1}
           {arriveTangent=0.0224, interpMode=RCIM_Cubic, leaveTangent=0.0224, tangentMode=RCTM_User, time=1, value=1}
          ]
         }}}
       Scale Curve=10
      )
      False Float=0
     )
    )
   ) {isSetParameters=true}
   &DynamicMaterialParameters(
    Index 0 Param 1=&FloatFromCurve(FloatCurve=DataInterface() {propertyValues {Curve {keys [<at 0 1.1>,<at 1 2>]}}}, Scale Curve=0.5)
    Index 0 Param 3=0.5
    Module.Index0Float0Write=true
    Module.Index0Float1Write=true
    Module.Index0Float2Write=true
    Module.Index0Float3Write=true
   )
   &VC_UpdateRibbon_v1()
   &Color(Color=&VC_DynamicScaleColor(
     Color to Scale:LinearColor=$Particles.Color
     RGB Scale:Vector3f=&VectorFromFloat(Value=&MakeCustomFloatFromBool(Bool=$User.bIsPainted, True Float=1, False Float=2))
     Alpha Scale:NiagaraFloat=1
    ))
  ]
  rendererClasses [&NiagaraRibbonRendererProperties]
  renderers [
   {rendererClass=&NiagaraRibbonRendererProperties}
   {rendererClass=&NiagaraRibbonRendererProperties, rendererIndex=1}
   {rendererClass=&NiagaraRibbonRendererProperties, rendererIndex=2}
   {rendererClass=&NiagaraRibbonRendererProperties, rendererIndex=3}
   {rendererClass=&NiagaraRibbonRendererProperties, rendererIndex=4}
  ]
 }
 NE_Ribbon_BoostCone {
  emitterUpdateScript [
   &EmitterState(Life Cycle Mode=::NewEnumerator1, Loop Duration=0.5, UseLoopDelay=false)
   &VC_FilteredSocketCount()
   &VC_RateQualityHandler() {enabled=false}
   &VC_SpawnRibbon_v1(SpawnRate=100)
  ]
  particleSpawnScript [
   &VC_PaintedVariantsOverride(
    Black=<rgba 0.077 0.077 0.077 2>
    BurntSienna=<rgba 0.2 0.059 0.02 1>
    Cobalt=<rgba 0.07 0.1397 1 1>
    Crimson=<rgba 0.35 0.0299 0.0245 1.5>
    DefaultColor=<rgba 2 2 2 0.75>
    ForestGreen=<rgba 0.0717 0.45 0.0135 1.5>
    Grey=<rgba 0.14 0.14 0.14 1.5>
    Lime=<rgba 0.3825 0.75 0.015 1>
    Orange=<rgba 0.5 0.125 0.05 1>
    Pink=<rgba 0.5017 0.105 0.7 1>
    Purple=<rgba 0.1002 0.021 0.7 1>
    Saffron=<rgba 0.35 0.35 0 1.5>
    SkyBlue=<rgba 0 1.075 1.5 0.5>
    TitaniumWhite=<rgba 1.5 1.5 1.5 0.7>
   )
   &VC_InitializeRibbon_v1(
    Color=$Particles.Color
    Custom Offset=true
    Lifetime=0.8
    Local Vector Offset=&AddVectorToPosition(
     Position=&AddVectorToPosition(Position=$User.DTCarOffset, Vector=Linked() {linkedVariable {name=User.SingleBoostPos, type=NiagaraPosition}})
     Vector=<30 0 5>
    )
    Offset Amount=1
    Ribbon Width=130
   )
   &AddVelocity(Velocity=&MakeCustomVectorFromBool(Bool=$User.bIsFrontendPreview, True Vector=<-2000 0 0>, False Vector=<-1500 0 0>)) {module=AddVelocity001}
  ]
  particleUpdateScript [
   &ParticleState()
   &ScaleColor(
    Scale Alpha=&FloatFromCurve(
     FloatCurve=DataInterface() {propertyValues {Curve {
        defaultValue=1
        keys [
         {arriveTangent=6.382, interpMode=RCIM_Cubic, leaveTangent=6.382, tangentMode=RCTM_User, time=0.077, value=0}
         {arriveTangent=0.1359, interpMode=RCIM_Cubic, leaveTangent=0.1359, tangentMode=RCTM_User, time=0.2, value=1}
         {arriveTangent=-0.0357, interpMode=RCIM_Cubic, leaveTangent=-0.0357, tangentMode=RCTM_User, time=0.4, value=1}
         {arriveTangent=-1.167, interpMode=RCIM_Cubic, leaveTangent=-1.167, tangentMode=RCTM_User, time=1, value=0}
        ]
       }}}
     Scale Curve=25
    )
    ScaleA=true
    ScaleRGB=true
   )
   &InheritSourceMovement(
    Applied Position Scale=&VectorFromFloat(Value=&FloatFromCurve(FloatCurve=DataInterface() {propertyValues {Curve {keys [
          {interpMode=RCIM_Cubic, tangentMode=RCTM_Break, time=0, value=1}
          {arriveTangent=-2.9025, interpMode=RCIM_Cubic, tangentMode=RCTM_Break, time=1, value=0}
         ]}}}))
    ApplyForce=false
    ApplyPosition=true
    ApplyPositionFalloff=false
   )
   &ScaleRibbonWidth(Ribbon Width Scale=&FloatFromCurve(FloatCurve=DataInterface() {propertyValues {Curve {keys [
         <at 0 0.2 RCIM_Cubic>
         {arriveTangent=-0.8211, interpMode=RCIM_Cubic, leaveTangent=-0.8211, tangentMode=RCTM_User, time=0.15, value=1}
         <at 1 0.3 RCIM_Cubic>
        ]}}}))
   &SolveForcesAndVelocity(Clamp Velocity=false, Limit Acceleration=false)
   &VC_UpdateRibbon_v1()
  ]
  rendererClasses [&NiagaraRibbonRendererProperties]
  renderers [{rendererClass=&NiagaraRibbonRendererProperties}]
 }
 NE_SkullMesh {
  emitterUpdateScript [
   &EmitterState(
    Inactive Response=::NewEnumerator1
    Life Cycle Mode=::NewEnumerator1
    Loop Behavior=::NewEnumerator1
    Loop Duration Mode=::NewEnumerator1
    UseLoopDelay=false
   )
   &SpawnBurst_Instantaneous(Spawn Group=3, Use Spawn Probability=false, UseLoopCountLimit=false) {module=SpawnBurst_Instantaneous003}
   &SpawnBurst_Instantaneous(Spawn Group=2, Use Spawn Probability=false, UseLoopCountLimit=false) {module=SpawnBurst_Instantaneous002}
   &SpawnBurst_Instantaneous(Use Spawn Probability=false, UseLoopCountLimit=false)
   &SpawnBurst_Instantaneous(Spawn Group=1, Use Spawn Probability=false, UseLoopCountLimit=false) {module=SpawnBurst_Instantaneous001}
   &SpawnRate(SpawnRate=2, Use Spawn Probability=false) {enabled=false}
   &SpawnRate(SpawnGroup=1, SpawnRate=2, Use Spawn Probability=false) {enabled=false, module=SpawnRate001}
   &SpawnRate(SpawnGroup=2, SpawnRate=2, Use Spawn Probability=false) {enabled=false, module=SpawnRate002}
   &SpawnRate(SpawnGroup=3, SpawnRate=2, Use Spawn Probability=false) {enabled=false, module=SpawnRate003}
  ]
  particleSpawnScript [
   &VC_PaintedVariantsOverride(
    Black=<rgba 0.1 0.1 0.1 2>
    BurntSienna=<rgba 0.3 0.043 0.009 2>
    Cobalt=<rgba 0.0875 0.1747 1.25 2>
    Crimson=<rgba 0.5 0 0.0083 2>
    ForestGreen=<rgba 0.1413 0.8 0.04 2>
    Grey=<rgba 0.5 0.5 0.5 1.5>
    Lime=<rgba 0.7575 1.5 0.015 1>
    Orange=<rgba 0.8 0.176 0.08 2>
    Pink=<rgba 0.95 0.19 0.9118 1.5>
    Purple=<rgba 0.36 0.075 1.5 2>
    R1 Black:LinearColor=<rgba 1 0.0833 0 1>
    R1 BurntSienna:LinearColor=<rgba 1 0.0833 0 1>
    R1 Cobalt:LinearColor=<rgba 1 0.0833 0 1>
    R1 Crimson:LinearColor=<rgba 1 0.0833 0 1>
    R1 DefaultColor=<rgba 1 0 0 1>
    R1 ForestGreen:LinearColor=<rgba 1 0.0833 0 0.8>
    R1 Grey:LinearColor=<rgba 1 0.0833 0 1>
    R1 Lime:LinearColor=<rgba 1 0.0833 0 1>
    R1 Orange:LinearColor=<rgba 1 0.0833 0 1>
    R1 Pink:LinearColor=<rgba 1 0.0833 0 1>
    R1 Purple:LinearColor=<rgba 1 0.0833 0 1>
    R1 Saffron:LinearColor=<rgba 1 0.0833 0 1>
    R1 SkyBlue:LinearColor=<rgba 1 0.0833 0 1>
    R1 TitaniumWhite:LinearColor=<rgba 0.6435 0.6435 0.6435 1>
    R2 Black:LinearColor=<rgba 0 1 0 1>
    R2 BurntSienna:LinearColor=<rgba 0 1 0 0.5>
    R2 Cobalt:LinearColor=<rgba 0 1 0 1>
    R2 Crimson:LinearColor=<rgba 0 1 0 0.5>
    R2 DefaultColor:LinearColor=<rgba 0 1 0 1>
    R2 ForestGreen:LinearColor=<rgba 0 1 0 1>
    R2 Grey:LinearColor=<rgba 0 1 0 1>
    R2 Lime:LinearColor=<rgba 0 1 0 1>
    R2 Orange:LinearColor=<rgba 0 1 0 0.8>
    R2 Pink:LinearColor=<rgba 0 1 0 0.8>
    R2 Purple:LinearColor=<rgba 0 1 0 0.5>
    R2 Saffron:LinearColor=<rgba 0 1 0 0.8>
    R2 SkyBlue:LinearColor=<rgba 0 1 0 0.8>
    R2 TitaniumWhite:LinearColor=<rgba 0.6435 0.6435 0.6435 1>
    R3 Black:LinearColor=<rgba 0 0 1 2>
    R3 BurntSienna:LinearColor=<rgba 0 0.3333 4 2>
    R3 Cobalt:LinearColor=<rgba 0 0.3333 4 2>
    R3 Crimson:LinearColor=<rgba 0 0.3333 4 2>
    R3 DefaultColor:LinearColor=<rgba 0 8.3333 50 0.5>
    R3 ForestGreen:LinearColor=<rgba 0 0.3333 4 3>
    R3 Grey:LinearColor=<rgba 0 0.3333 4 3>
    R3 Lime:LinearColor=<rgba 0 0.3333 4 1>
    R3 Orange:LinearColor=<rgba 0 0.3333 4 3>
    R3 Pink:LinearColor=<rgba 0 0.3333 4 3>
    R3 Purple:LinearColor=<rgba 0 0.25 3 3>
    R3 Saffron:LinearColor=<rgba 0 0.3333 4 3>
    R3 SkyBlue:LinearColor=<rgba 0 0.1667 2 5>
    R3 TitaniumWhite:LinearColor=<rgba 0.6435 0.6435 0.6435 1>
    R4 Black:LinearColor=<rgba 0.0191 0.0191 0.0191 1.5>
    R4 BurntSienna:LinearColor=<rgba 0.05 0.0095 0.0025 1>
    R4 Cobalt:LinearColor=<rgba 0.049 0.0978 0.7 2>
    R4 Crimson:LinearColor=<rgba 0.25 0.0214 0.0175 1>
    R4 DefaultColor:LinearColor=<rgba 0.3 0.3 0.3 1>
    R4 ForestGreen:LinearColor=<rgba 0.0717 0.45 0.0135 1.5>
    R4 Grey:LinearColor=<rgba 0.1 0.1 0.1 1>
    R4 Lime:LinearColor=<rgba 0.1575 0.3 0.015 1>
    R4 Orange:LinearColor=<rgba 0.5391 0.0899 0 1>
    R4 Pink:LinearColor=<rgba 0.1533 0.06 0.2 1>
    R4 Purple:LinearColor=<rgba 0.0812 0.0105 0.35 1>
    R4 Saffron:LinearColor=<rgba 0.2 0.2 0 1>
    R4 SkyBlue:LinearColor=<rgba 0 0.1583 0.25 1>
    R4 TitaniumWhite:LinearColor=<rgba 0.3 0.3 0.3 1>
    Saffron=<rgba 0.74 0.6265 0.0592 2>
    SkyBlue=<rgba 0.05 0.81 1 2>
    TitaniumWhite=<rgba 0.6435 0.6435 0.6435 2>
    Use R1 Painted Overrides=true
    Use R2 Painted Overrides=true
    Use R3 Painted Overrides=true
    Use R4 Painted Overrides=true
   )
   &InitializeParticle(
    Color Mode=::NewEnumerator0
    Lifetime=/VehicleCosmetics/FX/Niagara/Modules/MergeEmitters/VC_FloatBySpawnGroup(
     SpawnGroup or ExecIndex:NiagaraInt32=0
     CustomInt:NiagaraInt32=0
     MaximumSpawnGroups:NiagaraInt32=5
     Float SG 0:NiagaraFloat=1
     Float SG 1:NiagaraFloat=1
     Float SG 2:NiagaraFloat=1
     Float SG 3:NiagaraFloat=1
     Float SG 4:NiagaraFloat=2
     Float SG 5:NiagaraFloat=0
    )
    Mesh Scale Mode=::NewEnumerator3
    Mesh Uniform Scale=&VC_FloatBySpawnGroup(
     SpawnGroup or ExecIndex:NiagaraInt32=0
     CustomInt:NiagaraInt32=0
     MaximumSpawnGroups:NiagaraInt32=5
     Float SG 0:NiagaraFloat=0.3
     Float SG 1:NiagaraFloat=0.325
     Float SG 2:NiagaraFloat=0.32
     Float SG 3:NiagaraFloat=0.315
     Float SG 4:NiagaraFloat=0
     Float SG 5:NiagaraFloat=0
    )
    Position=&AddVectorToPosition(Position=$User.DTCarOffset, Vector=Linked() {linkedVariable {name=User.SingleBoostPos, type=NiagaraPosition}})
    Position Mode=::NewEnumerator1
    Position Offset=/VehicleCosmetics/FX/Niagara/Modules/MergeEmitters/VC_VectorBySpawnGroup(
     SpawnGroup or ExecIndex:NiagaraInt32=0
     CustomInt:NiagaraInt32=0
     MaximumSpawnGroups:NiagaraInt32=5
     Vector Group 0:Vector3f=<-100 0 0>
     Vector Group 1:Vector3f=<-100 0.6 2>
     Vector Group 2:Vector3f=<-98 -1 1.5>
     Vector Group 3:Vector3f=<-101 -0.4 -0.5>
     Vector Group 4:Vector3f=<-40 0 3>
     Vector Group 5:Vector3f=<0 0 0>
    )
    Position Offset Coordinate Space=::Local
    Sprite Rotation Mode=::NewEnumerator2
    Sprite Size Mode=::NewEnumerator3
    Uniform Sprite Size=220
    UsePositionOffset=true
    Write Material Random=true
   )
   &AddVelocity(Velocity=&VC_VectorBySpawnGroup(
     SpawnGroup or ExecIndex:NiagaraInt32=0
     CustomInt:NiagaraInt32=0
     MaximumSpawnGroups:NiagaraInt32=5
     Vector Group 0:Vector3f=<0 0 0>
     Vector Group 1:Vector3f=<0 0 0>
     Vector Group 2:Vector3f=<0 0 0>
     Vector Group 3:Vector3f=<0 0 0>
     Vector Group 4:Vector3f=&RandomRangeVector(Minimum=<-100 0 0>, Maximum=<-200 1 1>)
     Vector Group 5:Vector3f=<0 0 0>
    ))
   &Color(Color=$Particles.Color)
   /SetVariables(Particles.RenderVisTag:NiagaraInt32=&SpawnGroup()) {isSetParameters=true}
   &InitialMeshOrientation(
    Mesh Orientation Mode=::NewEnumerator7
    Rotation=&RandomRangeVector(Minimum=<0 -0.02 -0.02>, Maximum=<0 0.02 0.02>)
    Use Rotation Vector=true
   )
   &ShapeLocation(MaskSpawnGroup=false, Override Seed=false, Sphere Radius=1, UseOwnerScale=true)
  ]
  particleUpdateScript [
   &ParticleState(Kill Particles When Lifetime Has Elapsed=false, Loop Particles Lifetime=true)
   &ScaleSpriteSize(Uniform Curve Sprite Scale=DataInterface() {propertyValues {Curve {keys [
        {interpMode=RCIM_Cubic, tangentMode=RCTM_User, time=0, value=0}
        {interpMode=RCIM_Cubic, tangentMode=RCTM_User, time=0.19, value=0.75}
        {interpMode=RCIM_Cubic, tangentMode=RCTM_User, time=1, value=0}
       ]}}})
   &SolveForcesAndVelocity(Clamp Velocity=false, Limit Acceleration=false)
   /SetVariables(
    Particles.R1_SkullR_ScaledColor:LinearColor=&VC_DynamicScaleColor(
     Color to Scale:LinearColor=$Particles.Initial.R1 Color
     RGB Scale:Vector3f=<1 1 1>
     Alpha Scale:NiagaraFloat=&FloatFromCurve(FloatCurve=DataInterface() {propertyValues {Curve {keys [
          {interpMode=RCIM_Cubic, leaveTangentWeight=0.0333, tangentMode=RCTM_User, time=0, value=0.5}
          {arriveTangentWeight=0.0381, interpMode=RCIM_Cubic, leaveTangentWeight=0.3051, tangentMode=RCTM_User, time=0.1, value=1}
          {arriveTangentWeight=0.3051, interpMode=RCIM_Cubic, leaveTangentWeight=0.0381, tangentMode=RCTM_User, time=0.9, value=1}
          {arriveTangentWeight=0.0333, interpMode=RCIM_Cubic, tangentMode=RCTM_User, time=1, value=0.5}
         ]}}})
    )
    Particles.R2_SkullG_ScaledColor:LinearColor=&VC_DynamicScaleColor(
     Color to Scale:LinearColor=$Particles.Initial.R2 Color
     RGB Scale:Vector3f=<1 1 1>
     Alpha Scale:NiagaraFloat=&FloatFromCurve(FloatCurve=DataInterface() {propertyValues {Curve {
         defaultValue=1
         keys [
          {interpMode=RCIM_Cubic, tangentMode=RCTM_User, time=0, value=0}
          {interpMode=RCIM_Cubic, tangentMode=RCTM_User, time=0.35, value=0.5}
          {tangentMode=RCTM_User, time=1, value=0}
         ]
        }}})
    )
    Particles.R3_SkullB_ScaledColor:LinearColor=&VC_DynamicScaleColor(
     Color to Scale:LinearColor=$Particles.Initial.R3 Color
     RGB Scale:Vector3f=<1 1 1>
     Alpha Scale:NiagaraFloat=&FloatFromCurve(FloatCurve=DataInterface() {propertyValues {Curve {keys [
          {interpMode=RCIM_Cubic, leaveTangentWeight=0.0333, tangentMode=RCTM_User, time=0, value=0.5}
          {arriveTangentWeight=0.0381, interpMode=RCIM_Cubic, leaveTangentWeight=0.3051, tangentMode=RCTM_User, time=0.1, value=1}
          {arriveTangentWeight=0.3051, interpMode=RCIM_Cubic, leaveTangentWeight=0.0381, tangentMode=RCTM_User, time=0.9, value=1}
          {arriveTangentWeight=0.0333, interpMode=RCIM_Cubic, tangentMode=RCTM_User, time=1, value=0.5}
         ]}}})
    )
    Particles.R4_Smoke_ScaledColor:LinearColor=&VC_DynamicScaleColor(
     Color to Scale:LinearColor=$Particles.Initial.R4 Color
     RGB Scale:Vector3f=<1 1 1>
     Alpha Scale:NiagaraFloat=&FloatFromCurve(
      FloatCurve=DataInterface() {propertyValues {Curve {
         defaultValue=1
         keys [
          {interpMode=RCIM_Cubic, tangentMode=RCTM_User, time=0, value=0}
          {interpMode=RCIM_Cubic, tangentMode=RCTM_User, time=0.35, value=0.5}
          {arriveTangent=-0.1799, interpMode=RCIM_Cubic, leaveTangent=-0.1799, tangentMode=RCTM_User, time=0.5, value=0}
          {tangentMode=RCTM_User, time=1, value=0}
         ]
        }}}
      Scale Curve=0.3
     )
    )
   ) {isSetParameters=true}
   &DynamicMaterialParameters(
    Index 0 Param 1=&MakeCustomFloatFromBool(Bool=$User.bIsFrontendPreview, True Float=1, False Float=0)
    Index 0 Param 2=&MakeCustomFloatFromBool(Bool=$User.bIsFrontendPreview, True Float=0, False Float=1)
    Index 0 Param 3=5
    Index 1 Param 1=&FloatFromCurve(FloatCurve=DataInterface() {propertyValues {Curve {keys [
         {interpMode=RCIM_Cubic, tangentMode=RCTM_User, time=0, value=0}
         {interpMode=RCIM_Cubic, tangentMode=RCTM_User, time=0.25, value=1}
         {interpMode=RCIM_Cubic, tangentMode=RCTM_User, time=1, value=0.5}
        ]}}})
    Index 1 Param 2=&FloatFromCurve(FloatCurve=DataInterface() {propertyValues {Curve {keys [{interpMode=RCIM_Cubic, tangentMode=RCTM_User, time=0, value=1}, {interpMode=RCIM_Cubic, tangentMode=RCTM_User, time=0.5, value=0.1}]}}})
    Module.Index0Float0Write=true
    Module.Index0Float1Write=true
    Module.Index0Float2Write=true
    Module.Index0Float3Write=true
    Module.Index1Float0Write=true
    Module.Index1Float1Write=true
    Module.Index1Float2Write=true
    Module.Index1Float3Write=true
    Write Parameter Index 1=::NewEnumerator1
   )
   &ScaleColor(ScaleA=true, ScaleRGB=true) {module=ScaleColor001}
  ]
  rendererClasses [&NiagaraSpriteRendererProperties, &NiagaraMeshRendererProperties]
  renderers [
   {rendererClass=&NiagaraSpriteRendererProperties}
   {rendererClass=&NiagaraMeshRendererProperties, rendererIndex=1}
   {rendererClass=&NiagaraMeshRendererProperties, rendererIndex=2}
   {rendererClass=&NiagaraMeshRendererProperties, rendererIndex=3}
   {rendererClass=&NiagaraMeshRendererProperties, rendererIndex=4}
  ]
 }
 NE_SmokeBurst001 {
  emitterUpdateScript [
   &EmitterState(Inactive Response=::NewEnumerator1, Life Cycle Mode=::NewEnumerator1, Loop Duration=1, UseLoopDelay=false)
   &SpawnRate(SpawnGroup=1, SpawnRate=10, Use Spawn Probability=false)
   &SpawnBurst_Instantaneous(
    Spawn Count=&MakeCustomIntFromBool(Bool=$User.bIsFrontendPreview, True INT=0, False INT=20)
    Use Spawn Probability=false
    UseLoopCountLimit=true
   )
  ]
  particleSpawnScript [
   &VC_PaintedVariantsOverride(
    Black=<rgba 0.0191 0.0191 0.0191 1.5>
    BurntSienna=<rgba 0.05 0.0095 0.0025 1>
    Cobalt=<rgba 0.04 0.067 0.4 1.25>
    Crimson=<rgba 0.35 0.0299 0.0245 1.5>
    ForestGreen=<rgba 0.0717 0.45 0.0135 1.5>
    Grey=<rgba 0.14 0.14 0.14 1.5>
    Lime=<rgba 0.3569 0.5391 0.0054 1>
    Orange=<rgba 0.6724 0.2346 0 1>
    Pink=<rgba 0.45 0.135 0.4342 1>
    Purple=<rgba 0.0812 0.0105 0.35 1>
    R1 Black:LinearColor=<rgba 0.0191 0.0191 0.0191 1.5>
    R1 BurntSienna:LinearColor=<rgba 0.05 0.0095 0.0025 1>
    R1 Cobalt:LinearColor=<rgba 0.049 0.0978 0.7 2>
    R1 Crimson:LinearColor=<rgba 0.25 0.0214 0.0175 1>
    R1 DefaultColor=<rgba 0.3 0.3 0.3 1>
    R1 ForestGreen:LinearColor=<rgba 0.0717 0.45 0.0135 1.5>
    R1 Grey:LinearColor=<rgba 0.1 0.1 0.1 1>
    R1 Lime:LinearColor=<rgba 0.1575 0.3 0.015 1>
    R1 Orange:LinearColor=<rgba 0.5391 0.0899 0 1>
    R1 Pink:LinearColor=<rgba 0.1533 0.06 0.2 1>
    R1 Purple:LinearColor=<rgba 0.0812 0.0105 0.35 1>
    R1 Saffron:LinearColor=<rgba 0.2 0.2 0 1>
    R1 SkyBlue:LinearColor=<rgba 0 0.1583 0.25 1>
    R1 TitaniumWhite:LinearColor=<rgba 0.3 0.3 0.3 1>
    Saffron=<rgba 0.35 0.35 0 1>
    SkyBlue=<rgba 0.0237 0.3848 0.475 1>
    TitaniumWhite=<rgba 0.6435 0.6435 0.6435 1>
    Use R1 Painted Overrides=true
   )
   &InitializeParticle(
    Color Mode=::NewEnumerator0
    Lifetime=&VC_FloatBySpawnGroup(
     SpawnGroup or ExecIndex:NiagaraInt32=0
     CustomInt:NiagaraInt32=0
     MaximumSpawnGroups:NiagaraInt32=2
     Float SG 0:NiagaraFloat=$Emitter.CurrentLoopDuration
     Float SG 1:NiagaraFloat=2
     Float SG 2:NiagaraFloat=0
     Float SG 3:NiagaraFloat=0
     Float SG 4:NiagaraFloat=0
     Float SG 5:NiagaraFloat=0
    )
    Position=&AddVectorToPosition(Position=$User.DTCarOffset, Vector=Linked() {linkedVariable {name=User.SingleBoostPos, type=NiagaraPosition}})
    Position Mode=::NewEnumerator1
    Position Offset=&VC_VectorBySpawnGroup(
     SpawnGroup or ExecIndex:NiagaraInt32=0
     CustomInt:NiagaraInt32=0
     MaximumSpawnGroups:NiagaraInt32=2
     Vector Group 0:Vector3f=<-90 0 0>
     Vector Group 1:Vector3f=<-40 0 3>
     Vector Group 2:Vector3f=<0 0 0>
     Vector Group 3:Vector3f=<0 0 0>
     Vector Group 4:Vector3f=<0 0 0>
     Vector Group 5:Vector3f=<0 0 0>
    )
    Position Offset Coordinate Space=::Local
    Sprite Rotation Mode=::NewEnumerator2
    Sprite Size Mode=::NewEnumerator3
    Uniform Sprite Size=&VC_FloatBySpawnGroup(
     SpawnGroup or ExecIndex:NiagaraInt32=0
     CustomInt:NiagaraInt32=0
     MaximumSpawnGroups:NiagaraInt32=2
     Float SG 0:NiagaraFloat=&RandomRangeFloat(Minimum=70, Maximum=110)
     Float SG 1:NiagaraFloat=200
     Float SG 2:NiagaraFloat=0
     Float SG 3:NiagaraFloat=0
     Float SG 4:NiagaraFloat=0
     Float SG 5:NiagaraFloat=0
    )
    UsePositionOffset=true
    Write Material Random=true
   )
   &AddVelocity(
    Module.Use Falloff Distance=false
    Override Seed=false
    Velocity Mode=::NewEnumerator1
    Velocity Origin=$Particles.Position
    Velocity Speed=&VC_FloatBySpawnGroup(
     SpawnGroup or ExecIndex:NiagaraInt32=0
     CustomInt:NiagaraInt32=0
     MaximumSpawnGroups:NiagaraInt32=2
     Float SG 0:NiagaraFloat=&RandomRangeFloat(Minimum=80, Maximum=120)
     Float SG 1:NiagaraFloat=0
     Float SG 2:NiagaraFloat=0
     Float SG 3:NiagaraFloat=0
     Float SG 4:NiagaraFloat=0
     Float SG 5:NiagaraFloat=0
    )
   )
   &Color(Color=$Particles.Color, Scale Alpha=10)
   /SetVariables(Particles.RENDER_VISIBILITY:NiagaraInt32=&SpawnGroup()) {isSetParameters=true}
   &AddVelocity(Velocity=&VC_VectorBySpawnGroup(
     SpawnGroup or ExecIndex:NiagaraInt32=0
     CustomInt:NiagaraInt32=0
     MaximumSpawnGroups:NiagaraInt32=2
     Vector Group 0:Vector3f=<0 0 0>
     Vector Group 1:Vector3f=&RandomRangeVector(Minimum=<-100 0 -2>, Maximum=<-200 1 2>)
     Vector Group 2:Vector3f=<0 0 0>
     Vector Group 3:Vector3f=<0 0 0>
     Vector Group 4:Vector3f=<0 0 0>
     Vector Group 5:Vector3f=<0 0 0>
    )) {module=AddVelocity001}
   &ShapeLocation(
    MaskSpawnGroup=false
    Non Uniform Scale=&VC_VectorBySpawnGroup(
     SpawnGroup or ExecIndex:NiagaraInt32=0
     CustomInt:NiagaraInt32=0
     MaximumSpawnGroups:NiagaraInt32=2
     Vector Group 0:Vector3f=<0 0 0>
     Vector Group 1:Vector3f=<1 1 1>
     Vector Group 2:Vector3f=<0 0 0>
     Vector Group 3:Vector3f=<0 0 0>
     Vector Group 4:Vector3f=<0 0 0>
     Vector Group 5:Vector3f=<0 0 0>
    )
    Override Seed=false
    Sphere Radius=10
    UseOwnerScale=true
   )
  ]
  particleUpdateScript [
   &ParticleState()
   &SolveForcesAndVelocity(Clamp Velocity=false, Limit Acceleration=false)
   &ScaleSpriteSize(Uniform Curve Sprite Scale=DataInterface() {propertyValues {Curve {keys [
        {interpMode=RCIM_Cubic, tangentMode=RCTM_User, time=0, value=0}
        {interpMode=RCIM_Cubic, tangentMode=RCTM_User, time=0.19, value=0.7}
        {interpMode=RCIM_Cubic, tangentMode=RCTM_User, time=1, value=0}
       ]}}})
   &DynamicMaterialParameters(
    Index 1 Param 1=&FloatFromCurve(FloatCurve=DataInterface() {propertyValues {Curve {keys [
         {interpMode=RCIM_Cubic, tangentMode=RCTM_User, time=0, value=0}
         {interpMode=RCIM_Cubic, tangentMode=RCTM_User, time=0.25, value=1}
         {interpMode=RCIM_Cubic, tangentMode=RCTM_User, time=1, value=0.5}
        ]}}})
    Index 1 Param 2=&FloatFromCurve(FloatCurve=DataInterface() {propertyValues {Curve {keys [{interpMode=RCIM_Cubic, tangentMode=RCTM_User, time=0, value=1}, {interpMode=RCIM_Cubic, tangentMode=RCTM_User, time=0.5, value=0.1}]}}})
    Module.Index1Float0Write=true
    Module.Index1Float1Write=true
    Module.Index1Float2Write=true
    Module.Index1Float3Write=true
    Write Parameter Index 0=::NewEnumerator0
    Write Parameter Index 1=::NewEnumerator1
   )
   &ScaleColor(
    Scale Alpha=&FloatFromCurve(
     FloatCurve=DataInterface() {propertyValues {Curve {
        defaultValue=1
        keys [
         {interpMode=RCIM_Cubic, tangentMode=RCTM_User, time=0, value=0}
         {interpMode=RCIM_Cubic, tangentMode=RCTM_User, time=0.35, value=0.5}
         {tangentMode=RCTM_User, time=1, value=0}
        ]
       }}}
     Scale Curve=0.3
    )
    ScaleA=true
    ScaleRGB=true
   )
   /SetVariables(Particles.R1_Smoke_Scalecolor:LinearColor=&VC_DynamicScaleColor(
     Color to Scale:LinearColor=$Particles.Initial.R1 Color
     RGB Scale:Vector3f=<1 1 1>
     Alpha Scale:NiagaraFloat=&FloatFromCurve(
      FloatCurve=DataInterface() {propertyValues {Curve {
         defaultValue=1
         keys [
          {interpMode=RCIM_Cubic, tangentMode=RCTM_User, time=0, value=0}
          {interpMode=RCIM_Cubic, tangentMode=RCTM_User, time=0.35, value=0.5}
          {arriveTangent=-0.1799, interpMode=RCIM_Cubic, leaveTangent=-0.1799, tangentMode=RCTM_User, time=0.5, value=0}
          {tangentMode=RCTM_User, time=1, value=0}
         ]
        }}}
      Scale Curve=&MakeCustomFloatFromBool(Bool=$User.bIsFrontendPreview, True Float=0.3, False Float=1)
     )
    )) {isSetParameters=true}
  ]
  rendererClasses [&NiagaraSpriteRendererProperties]
  renderers [
   {rendererClass=&NiagaraSpriteRendererProperties}
   {rendererClass=&NiagaraSpriteRendererProperties, rendererIndex=1}
  ]
 }
}
summary {
 systemName=NS_Boost_SkullChromatic
 emitters [
  {emitterName=NE_SkullMesh, rendererClasses [&NiagaraSpriteRendererProperties, &NiagaraMeshRendererProperties]}
  {emitterName=NE_Ribbon_BoostCone, rendererClasses [&NiagaraRibbonRendererProperties]}
  {emitterName=NE_MainRibbon002, rendererClasses [&NiagaraRibbonRendererProperties]}
  {emitterName=NE_SmokeBurst001, rendererClasses [&NiagaraSpriteRendererProperties]}
 ]
 userVariables [
  User.bIsFrontendPreview {type=NiagaraBool}
  User.bIsPainted {type=NiagaraBool}
  User.ColorVariantID {type=NiagaraInt32}
  User.CurrentPaintedColor {type=LinearColor, description="", defaultValue {struct=LinearColor, value=<rgba 0.05 0.05 0.05 1.5>}}
  User.DTCarOffset {type=NiagaraPosition, description="", defaultValue {}}
  User.GameModeID {type=NiagaraInt32}
  User.PaintedColor {type=LinearColor, description="", defaultValue {struct=LinearColor, value=<rgba 1 1 1 1>}}
  User.PaintedScalar {type=NiagaraFloat, defaultValue=1, description=""}
  User.PortCount {type=NiagaraInt32, defaultValue=1}
  User.SingleBoostPos {type=NiagaraPosition, description="", defaultValue {}}
 ]
}
