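# Translucent unlit Niagara ribbon material for the cone trail of the SkullChromatic vehicle boost.
# A panning smoke texture, shaped by Hardness and particle alpha, builds a saturated opacity mask.
# Ceil on V*(1-V) hard-cuts the ribbon to a central band, while two power curves on U fade it along its length.
# Particle color tints panning noise plus the rerouted smoke, fed into MF_FX_EmissiveAdjustments as emissive.
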Material /VehicleCosmetics/Boosts/Boost_SkullChromatic/Materials/M_BoostSkullChromatic_ConeTrail
graph {
 properties {
  bAutomaticallySetUsageInEditor=false
  bUsedWithNiagaraRibbons=true
  blendMode=BLEND_Translucent
  shadingModel=MSM_Unlit
  usedShadingModels=Unlit
 }
 nodes [
  n1 = Time()
  n2 = ScalarParameter() {defaultValue=1.3, parameterName=SpeedM}
  n3 = TextureSample(UVs=Panner(Time=Multiply(A=@n1, B=@n2)) {speedX=-0.6}) {texture=/DealershipSharedLibrary/FX/Textures/Gradient_Masks/T_VCVFX_Smoke_Trail_A}
  n4 = Saturate(None=Multiply(A=Constant() {r=50}, B=@n3.R))
  n5 = Power(Base=Multiply(A=@n4, B=@n3.R), Exp=ScalarParameter() {defaultValue=0.6, parameterName=Hardness})
  n6 = ComponentMask(None=TextureCoordinate()) {g=true}
  n7 = Multiply(A=@n6, B=OneMinus(None=@n6))
  n8 = Multiply(A=Multiply(A=ParticleColor().A, B=Multiply(A=@n4, B=@n5)), B=Ceil(None=Subtract(A=@n7) {constB=0.11}))
  n9 = Power(Base=ComponentMask(None=TextureCoordinate()) {r=true}, Exp=ScalarParameter() {defaultValue=0.5, parameterName=Alpha1})
  n10 = Power(Base=ComponentMask(None=TextureCoordinate()) {r=true}, Exp=ScalarParameter() {defaultValue=0.1, parameterName=Alpha2})
  MaterialExpressionConstantClamp_0 = Saturate(None=Multiply(A=Multiply(A=@n8, B=@n9), B=OneMinus(None=@n10)))
  n11 = Time()
  n12 = ScalarParameter() {defaultValue=0.3, parameterName=Speed2nd}
  n13 = TextureSample(UVs=Panner(Time=Multiply(A=@n11, B=@n12)) {speedX=-0.3}) {texture=/DealershipSharedLibrary/FX/Textures/Noise/T_VCVFX_Noise1_A}
  n14 = ParticleColor()
  n15 = NamedRerouteUsage()
  MaterialFunctionCall(Emissive=Add(A=Multiply(A=@n13.G, B=@n14.RGB), B=Multiply(A=@n14.RGB, B=@n15))) {materialFunction=&MF_FX_EmissiveAdjustments}
  NamedRerouteDeclaration(None=@n3.R) {name=PanningSmoke, nodeColor=<rgba 1 0.1304 0 1>}
 ]
}
