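# Material function in the Boost_SkullChromatic vehicle boost cosmetic that outputs the object's pivot location in world space as ObjectPivotLocation.
# It transforms a constant origin point from local to world space with TransformPosition, then masks the result to RGB.
# It takes no inputs and derives the pivot from the transform alone, without using an Object Position node.
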MaterialFunction /VehicleCosmetics/Boosts/Boost_SkullChromatic/Materials/MF_ObjectPivotPoint
graph {nodes [
  n1 = Constant4Vector() {constant=<rgba 0 0 0 1>}
  FunctionOutput(None=ComponentMask(None=TransformPosition(None=@n1.RGBA) {transformType=TRANSFORMPOSSOURCE_World}) {b=true, g=true, r=true}) {outputName=ObjectPivotLocation}
 ]}
