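# Translucent lit skull material for Niagara mesh particles in the SkullChromatic vehicle boost.
# Skull textures, Fresnel and panning noise, tinted by particle color, combine with procedural fire-eye masks driven by TurnOnEyes and IsPainted dynamic parameters.
# An "RGB?" static switch selects an alternate hat-textured skull look with inverted Fresnel.
# A WPO animation branch swings the jaw through vertex-green-masked rotation and adds layered sine drift.
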Material /VehicleCosmetics/Boosts/Boost_SkullChromatic/Materials/M_BoostSkullChromatic_Skull
graph {
 properties {
  bAutomaticallySetUsageInEditor=false
  bUsedWithNiagaraMeshParticles=true
  blendMode=BLEND_Translucent
  refractionMethod=RM_PixelNormalOffset
  usedShadingModels=Default Lit
 }
 nodes [
  n1 = ScalarParameter() {defaultValue=3, parameterName=SkullPaintedTextureLightness}
  MaterialExpressionConstantClamp_1 = Saturate(None=Multiply(A=@n1, B=TextureSample() {texture=/DealershipSharedLibrary/FX/Textures/Diffuse/T_VCVFX_Skullfire_D}.R))
  n2 = Power(Base=@MaterialExpressionConstantClamp_1, Exp=ScalarParameter() {defaultValue=0.5, parameterName=SkullOpacity})
  n3 = ScalarParameter() {defaultValue=0.5, group=Fresnel, parameterName=FresnelMul}
  n4 = Fresnel(Normal=Transform(None=TextureSample() {samplerType=SAMPLERTYPE_Normal, texture=/DealershipSharedLibrary/FX/Textures/Normal/T_VCVFX_Hat_Skullfire_N}.RGB)) {exponent=3}
  n5 = Multiply(A=TextureSample() {texture=&T_VCVFX_Skullfire_D}.G, B=@n4)
  n6 = Constant() {r=8}
  n7 = OneMinus(None=@n4)
  n8 = ScalarParameter() {defaultValue=5, parameterName=FresnelIntensity}
  n9 = NamedRerouteDeclaration(None=Add(A=Power(Base=@n5, Exp=@n6), B=Power(Base=@n7, Exp=@n8))) {name=FresnelOP, desc=FresnelOP, nodeColor=<rgba 0.0994 0.0034 1 1>}
  n10 = TextureSample(UVs=Panner() {speedY=0.2}) {samplerType=SAMPLERTYPE_LinearColor, texture=/DealershipSharedLibrary/FX/Textures/Noise/T_VCVFX_NoiseClouds}
  n11 = Constant() {r=0.5}
  n12 = Constant() {r=1.5}
  n13 = Add(A=TextureCoordinate() {uTiling=4, vTiling=4}, B=Constant() {r=-1})
  n14 = Add(A=Constant2Vector() {g=0.5, r=-1.45}, B=@n13)
  MaterialExpressionConstantClamp_6 = Saturate(None=OneMinus(None=DotProduct(A=@n14, B=@n14)))
  MaterialExpressionConstantClamp_4 = Saturate(None=ConstantBiasScale(None=@MaterialExpressionConstantClamp_6) {bias=-0.95, scale=10})
  n15 = Add(A=@n13, B=Constant2Vector() {g=0.5, r=-2.45})
  MaterialExpressionConstantClamp_7 = Saturate(None=OneMinus(None=DotProduct(A=@n15, B=@n15)))
  MaterialExpressionConstantClamp_5 = Saturate(None=ConstantBiasScale(None=@MaterialExpressionConstantClamp_7) {bias=-0.95, scale=10})
  MaterialExpressionConstantClamp_8 = Saturate(None=ConstantBiasScale(None=@MaterialExpressionConstantClamp_6) {bias=-0.98, scale=16})
  MaterialExpressionConstantClamp_9 = Saturate(None=ConstantBiasScale(None=@MaterialExpressionConstantClamp_7) {bias=-0.98, scale=16})
  n16 = Add(
   A=Add(A=@MaterialExpressionConstantClamp_4, B=@MaterialExpressionConstantClamp_5)
   B=Add(A=@MaterialExpressionConstantClamp_8, B=@MaterialExpressionConstantClamp_9)
  )
  MaterialExpressionConstantClamp_3 = Saturate(None=@n16)
  MaterialExpressionConstantClamp_2 = Saturate(None=OneMinus(None=Multiply(A=@n12, B=@MaterialExpressionConstantClamp_3)))
  n17 = Add(A=Add(A=@n2, B=Multiply(A=@n3, B=@n9)), B=Multiply(A=Multiply(A=@n10.RGB, B=@n11), B=@MaterialExpressionConstantClamp_2))
  MaterialExpressionMeshEmitterVertexColor_0 = ParticleColor()
  n18 = Multiply(A=Multiply(A=@n2, B=@n17), B=@MaterialExpressionMeshEmitterVertexColor_0.RGB)
  n19 = TextureSample() {texture=&T_VCVFX_Skullfire_D}
  n20 = Multiply(A=@MaterialExpressionMeshEmitterVertexColor_0.RGB, B=Multiply(A=@n17, B=Saturate(None=@n19.G))) {constB=0.75}
  n21 = Abs(None=Add(A=TextureCoordinate() {uTiling=4, vTiling=5}, B=Constant2Vector() {g=-0.1, r=-3}))
  n22 = Panner(Coordinate=MaterialFunctionCall(UVs=OneMinus(None=@n21)) {materialFunction=&RadialGradientExponential}.RadialGradientExponential) {speedX=1}
  n23 = Multiply(
   A=ScalarParameter() {defaultValue=3, parameterName=FireEyesBrightness}
   B=Multiply(A=@n16, B=TextureSample(UVs=@n22) {texture=/DealershipSharedLibrary/FX/Textures/TexturePackage/T_GradientMask_RGB}.RGB)
  )
  MaterialExpressionMeshEmitterVertexColor_2 = ParticleColor()
  n24 = Multiply(A=Multiply(A=Desaturation(None=@n23), B=@MaterialExpressionMeshEmitterVertexColor_2.RGB), B=Constant() {r=3})
  n25 = DynamicParameter() {defaultValue=<rgba 0 1 1 1>, paramNames [IsPainted,TurnOnEyes,Param3,Param4]}
  n26 = LinearInterpolate(
   A=Add(A=@n18, B=Multiply(A=@n20) {constB=0.75})
   B=Add(A=@n18, B=LinearInterpolate(A=@n23, B=@n24, Alpha=@n25.IsPainted))
   Alpha=DynamicParameter() {defaultValue=<rgba 0 1 1 1>, paramNames [IsPainted,TurnOnEyes,Param3,Param4]}.TurnOnEyes
  )
  MaterialExpressionConstantClamp_0 = Saturate(None=@n26)
  MaterialFunctionCall(Emissive=StaticSwitchParameter(True=NamedRerouteUsage(), False=@MaterialExpressionConstantClamp_0) {parameterName=RGB?}) {materialFunction=&MF_FX_EmissiveAdjustments}
  n27 = MaterialFunctionCall(Invert Fresnel=StaticBool() {value=true}, Power=Constant() {r=1.5}) {materialFunction=&Fresnel_Function}
  n28 = Multiply(A=TextureSample() {texture=/DealershipSharedLibrary/FX/Textures/Diffuse/T_VCVFX_Hat_Skullfire_D}.R, B=@n1)
  MaterialExpressionMeshEmitterVertexColor_1 = ParticleColor()
  n29 = Multiply(A=@n28, B=@MaterialExpressionMeshEmitterVertexColor_1.RGB)
  NamedRerouteDeclaration(None=Multiply(A=Power(Base=@n27.Result) {constExponent=1.5}, B=@n29)) {name=RGB Skull Emissive, nodeColor=<rgba 0 1 0.7162 1>}
  n30 = ObjectRadius()
  n31 = Transform(None=Constant3Vector() {constant=<rgba 0 -1 0 0>}) {transformSourceType=TRANSFORMSOURCE_Local}
  n32 = Multiply(A=Time(), B=ScalarParameter() {defaultValue=0.75, parameterName=Spinner_Speed})
  n33 = RotateAboutAxis(
   NormalizedRotationAxis=Normalize(VectorInput=Multiply(A=@n30, B=@n31))
   RotationAngle=LinearInterpolate(
    A=ScalarParameter() {defaultValue=-0.04, parameterName=MaxOpenMouth}
    B=ScalarParameter() {parameterName=MinOpenMouth}
    Alpha=Sine(None=@n32) {period=2}
   )
   PivotPoint=MaterialFunctionCall() {materialFunction=&/MF_ObjectPivotPoint}.ObjectPivotLocation
   Position=WorldPosition().XYZ
  )
  n34 = VertexColor()
  n35 = Multiply(A=ScalarParameter() {defaultValue=1.5, parameterName=SpeedDB}, B=Time())
  n36 = Multiply(A=ScalarParameter() {defaultValue=1, parameterName=SpeedAB}, B=Time())
  n37 = Multiply(
   A=LinearInterpolate(A=Constant() {r=1}, B=Sine(None=@n35), Alpha=Sine(None=@n36))
   B=ScalarParameter() {defaultValue=1, parameterName=X_Drift_Mult}
  )
  n38 = Multiply(A=ScalarParameter() {defaultValue=1.5, parameterName=SpeedD}, B=Time())
  n39 = Multiply(A=ScalarParameter() {defaultValue=1, parameterName=SpeedA}, B=Time())
  n40 = Multiply(
   A=LinearInterpolate(A=Constant() {r=1}, B=Sine(None=@n38), Alpha=Sine(None=@n39))
   B=ScalarParameter() {defaultValue=1, parameterName=Y_Drift_Mult}
  )
  n41 = ScalarParameter() {defaultValue=1.5, parameterName=SpeedC}
  n42 = Time()
  n43 = ScalarParameter() {defaultValue=1, parameterName=SpeedB}
  n44 = Add(A=Time(), B=ScalarParameter() {defaultValue=1, parameterName=TimeOffset})
  n45 = LinearInterpolate(A=Sine(None=Multiply(A=@n41, B=@n42)), B=Constant() {r=-1}, Alpha=Cosine(None=Multiply(A=@n43, B=@n44)))
  n46 = ScalarParameter() {defaultValue=1, parameterName=Z_Drift_Mult}
  n47 = Multiply(
   A=AppendVector(A=AppendVector(A=@n37, B=@n40) {desc=XY}, B=Multiply(A=@n45, B=@n46)) {desc=XYZ}
   B=ScalarParameter() {defaultValue=0.5, parameterName=Uniform_Drift_Mult}
  )
  NamedRerouteDeclaration(None=Add(A=Multiply(A=@n33, B=@n34.G) {desc=JawRotation}, B=ComponentMask(None=@n47) {b=true, g=true, r=true})) {name=WPO_Animation, nodeColor=<rgba 0.4221 0 1 1>}
  n48 = ParticleColor()
  NamedRerouteDeclaration(None=Multiply(A=Multiply(A=@n48.A, B=@n28), B=@n27.Result)) {name=RGB Skull Opacity, nodeColor=<rgba 1 0 0.4 1>}
  NamedRerouteDeclaration(None=@n29) {name=RGB Skull BaseColor, nodeColor=<rgba 1 0.0184 0 1>}
  NamedRerouteUsage()
  StaticSwitchParameter(True=NamedRerouteUsage(), False=@n26) {parameterName=RGB?}
  n49 = Add(A=NamedRerouteUsage(), B=Multiply(A=TextureSample() {texture=&T_VCVFX_Skullfire_D}.G) {constB=5})
  n50 = ParticleColor()
  StaticSwitchParameter(True=NamedRerouteUsage(), False=Saturate(None=Multiply(A=@n49, B=@n50.A))) {parameterName=RGB?}
  n51 = TextureSample() {samplerType=SAMPLERTYPE_Normal, texture=&T_VCVFX_Hat_Skullfire_N}
  StaticSwitchParameter(True=@n51.RGB, False=Multiply(A=@n51.RGB, B=Constant() {r=0.3})) {parameterName=RGB?}
 ]
}
