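# Translucent unlit Niagara sprite material for the smoke cone of the Skull Chromatic vehicle boost.
# Slowly rotating smoke noise distorts the UVs of a panning dust texture and a fire-cloud mask.
# The contrast-shaped dust, offset by the NoiseAmount dynamic parameter, multiplies the cloud into a clamped smoke mask.
# That mask drives particle-colored emissive scaled by Brightness through MF_FX_EmissiveAdjustments, and a second clamped mask weighted by particle alpha.
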Material /VehicleCosmetics/Boosts/Boost_SkullChromatic/Materials/M_BoostSkullChromtic_SmokeCone
graph {
 properties {
  bAutomaticallySetUsageInEditor=false
  bUsedWithNiagaraSprites=true
  blendMode=BLEND_Translucent
  shadingModel=MSM_Unlit
  usedShadingModels=Unlit
 }
 nodes [
  n1 = Rotator(Coordinate=TextureCoordinate() {uTiling=0.5, vTiling=0.5}, Time=Multiply(A=Time(), B=Constant() {r=0.5})) {speed=-0.25}
  n2 = Multiply(
   A=ComponentMask(None=TextureSample(UVs=@n1) {texture=/DealershipSharedLibrary/FX/Textures/Noise/T_VCVFX_Noise_Smoke_03_RGB}.RGB) {g=true, r=true}
   B=Constant() {r=0.08}
  )
  n3 = TextureCoordinate()
  n4 = TextureSample(UVs=Panner(Coordinate=Add(A=@n2, B=@n3)) {speedX=0.1, speedY=0.5}) {texture=/DealershipSharedLibrary/FX/Textures/Gradient_Masks/T_VCVFX_Dust_A}
  n5 = Add(A=@n4.R, B=Constant() {r=-0.2})
  n6 = Constant() {r=0.5}
  n7 = Multiply(A=@n4.RGB, B=Add(A=@n4.RGB, B=Multiply(A=@n5, B=@n6)))
  n8 = DynamicParameter() {parameterIndex=1, paramNames [Brightness,NoiseAmount,Param3,Param4]}
  n9 = Time()
  n10 = Constant() {r=0.5}
  n11 = TextureSample(UVs=Rotator(Coordinate=TextureCoordinate() {uTiling=0.5, vTiling=0.5}, Time=Multiply(A=@n9, B=@n10)) {speed=-0.25}) {texture=&T_VCVFX_Noise_Smoke_03_RGB}
  n12 = Add(A=Multiply(A=ComponentMask(None=@n11.RGB) {g=true, r=true}, B=Constant() {r=0.08}), B=TextureCoordinate())
  MaterialExpressionConstantClamp_5 = Saturate(None=Multiply(
    A=Add(A=@n7, B=@n8.NoiseAmount)
    B=TextureSample(UVs=@n12) {texture=/DealershipSharedLibrary/FX/Textures/Gradient_Masks/T_VCVFX_FireCloud02_A}.R
   ))
  n13 = Multiply(A=@MaterialExpressionConstantClamp_5, B=Constant() {r=3})
  MaterialExpressionVertexColor_2 = ParticleColor()
  n14 = Multiply(A=@n13, B=@MaterialExpressionVertexColor_2.A)
  n15 = Constant() {r=-0.01}
  MaterialExpressionConstantClamp_4 = Saturate(None=Multiply(A=Constant() {r=2}, B=Add(A=@n14, B=@n15)))
  n16 = ParticleColor()
  MaterialFunctionCall(Emissive=Multiply(A=DynamicParameter() {parameterIndex=1, paramNames [Brightness,NoiseAmount,Param3,Param4]}.Brightness, B=Multiply(A=@n16.RGB, B=@n13))) {materialFunction=&MF_FX_EmissiveAdjustments}
 ]
}
